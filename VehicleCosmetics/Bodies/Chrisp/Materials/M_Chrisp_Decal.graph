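# Masked, two-sided Default Lit material for decals on the Chrisp vehicle body.
# It holds two texture samples, the Chrisp M3 decal texture and a normal map, with no parameters or math nodes.
# Usage is flagged for skeletal meshes and mesh deformers, with automatic usage setting in the editor turned off.
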Material /VehicleCosmetics/Bodies/Chrisp/Materials/M_Chrisp_Decal
graph {
 properties {
  bAutomaticallySetUsageInEditor=false
  bUsedWithMeshDeformer=true
  bUsedWithSkeletalMesh=true
  blendMode=BLEND_Masked
  twoSided=true
  usedShadingModels=Default Lit
 }
 nodes [
  TextureSample() {texture=/VehicleCosmetics/Bodies/Chrisp/Textures/Chrisp_M3_decal}
  TextureSample() {samplerType=SAMPLERTYPE_Normal, texture=/VehicleCosmetics/Bodies/Chrisp/Textures/NormalMap}
 ]
}
